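# Physics asset for the LemonLime1H pickaxe cosmetic mesh, covering the handle and its dangling ribbon bones.
# Three capsules on the handle bone approximate the shaft and curved end, while a small capsule marks the ribbon master bone.
# Box bodies form two ribbon branches from that master, a two-segment chain and a single segment.
# Constraints leave all swing and twist axes free, and none links the master to the handle.
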PhysicsAsset /BRCosmetics/Weapons/FORT_Melee/Pickaxe_LemonLime1H/Meshes/Pickaxe_LemonLime1H_Physics
bodies {
 Melee_TwoHanded_Handle [
  {shapeType=Capsule, length=49.0206, radius=2.4829, center=<0.0951 -8.3043 41.5886>}
  {shapeType=Capsule, length=10.5035, radius=2.25, center=<4.9581 -8.0811 60.9658>, rotation=<rot -33.3278 0 0>}
  {shapeType=Capsule, length=11.5206, radius=2.4829, center=<12.2245 -8.1898 67.4821>, rotation=<rot -76.5801 0 0>}
 ]
 dyn_ribbon_a [{shapeType=Box, extentX=2.1, extentY=1.1, extentZ=6, center=<2.7659 0 0>, rotation=<rot 90 0 0>}]
 dyn_ribbon_a2 [{shapeType=Box, extentX=2.1, extentY=1, extentZ=8.5, center=<4.2595 0 0>, rotation=<rot 90 0 0>}]
 dyn_ribbon_b [{shapeType=Box, extentX=1.9118, extentY=0.2, extentZ=4, center=<2 0 0>, rotation=<rot 90 0 0>}]
 dyn_ribbon_mstr [{shapeType=Capsule, length=1.1, radius=1.1}]
}
constraints [
 {
  bone1Name=dyn_ribbon_a
  bone2Name=dyn_ribbon_mstr
  swing1LimitDegrees=45
  swing1Motion=Free
  swing2LimitDegrees=45
  swing2Motion=Free
  twistLimitDegrees=45
  twistMotion=Free
 }
 {
  bone1Name=dyn_ribbon_b
  bone2Name=dyn_ribbon_mstr
  swing1LimitDegrees=45
  swing1Motion=Free
  swing2LimitDegrees=45
  swing2Motion=Free
  twistLimitDegrees=45
  twistMotion=Free
 }
 {
  bone1Name=dyn_ribbon_a2
  bone2Name=dyn_ribbon_a
  swing1LimitDegrees=45
  swing1Motion=Free
  swing2LimitDegrees=45
  swing2Motion=Free
  twistLimitDegrees=45
  twistMotion=Free
 }
]
